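# Black is a Default Lit material in the demo folder of the VinderTech Moth glider.
# It holds two constant color nodes, one near-black and one mid grey, with no connections shown to each other or to material outputs.
# It has no textures or parameters and is flagged for use on skeletal meshes, Nanite and mesh deformers.
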Material /Game/Gadgets/Assets/VinderTech_GliderChute/Glider_Moth/DEMO/Black
graph {
 properties {
  bUsedWithMeshDeformer=true
  bUsedWithNanite=true
  bUsedWithSkeletalMesh=true
  pixelDepthOffsetMode=PDOM_Legacy
  usedShadingModels=Default Lit
 }
 nodes [
  Constant3Vector() {constant=<rgba 0.0156 0.0156 0.0156 1>}
  Constant3Vector() {constant=<rgba 0.4427 0.4427 0.4418 1>}
 ]
}
